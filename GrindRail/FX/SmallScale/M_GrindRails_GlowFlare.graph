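# Additive unlit glow flare material for small-scale grind rail effects, usable on Niagara sprites and ribbons.
# Particle color tints a soft flare texture, then passes through the shared FX emissive adjustment function.
# Particle alpha, the texture's green channel and a near-camera fade combine into opacity.
# A shading path switch adds depth fade to opacity on the default path and skips it on mobile.
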Material /GrindRail/FX/SmallScale/M_GrindRails_GlowFlare
graph {
 properties {
  bUsedWithNiagaraRibbons=true
  bUsedWithNiagaraSprites=true
  blendMode=BLEND_Additive
  pixelDepthOffsetMode=PDOM_Legacy
  shadingModel=MSM_Unlit
  usedShadingModels=Unlit
 }
 nodes [
  n1 = ParticleColor()
  n2 = TextureSample() {texture=/Game/Global/Textures/Default/Masks_and_Gradients/T_New_Soft_Flair}
  MaterialFunctionCall(Emissive=Multiply(A=@n1.RGB, B=@n2.RGB)) {materialFunction=&MF_FX_EmissiveAdjustments}
  n3 = Multiply(
   A=Multiply(A=@n1.A, B=@n2.G)
   B=MaterialFunctionCall(Distance=ScalarParameter() {defaultValue=600, group=Alpha, parameterName=NearFadeDistance, sortPriority=3}) {materialFunction=&Near_Camera_Fade}.Result
  )
  ShadingPathSwitch(Default=DepthFade(Opacity=@n3, FadeDistance=ScalarParameter() {defaultValue=128, parameterName=FadeDistance}), Mobile=@n3) {desc=Migrated from FeatureLevelSwitch}
 ]
}
